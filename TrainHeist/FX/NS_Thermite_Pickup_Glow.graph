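# Niagara system for a background glow effect, named for the thermite pickup.
# Its single BG_Glow emitter burst-spawns particles drawn by two mesh renderers, with random material values written at initialization.
# Particles are not killed when their lifetime elapses, so the glow persists.
# Mesh size is scaled every update by the User.GlobalScale parameter, converted from float to vector, allowing overall size control from outside the system.
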NiagaraSystem /TrainHeist/FX/NS_Thermite_Pickup_Glow
dependencies {usedDataInterfaces [], usedDynamicInputs [], usedModules [&SystemState]}
emitters {BG_Glow {
  emitterUpdateScript [
   &EmitterState()
   &SpawnBurst_Instantaneous(Use Spawn Probability=false, UseLoopCountLimit=false)
  ]
  particleSpawnScript [&InitializeParticle(
    Lifetime=5
    Mesh Scale Mode=::NewEnumerator3
    Sprite Size Mode=::NewEnumerator3
    Uniform Sprite Size=120
    UsePositionOffset=false
    Write Material Random=true
   )]
  particleUpdateScript [
   &ParticleState(Kill Particles When Lifetime Has Elapsed=false)
   &ScaleMeshSize(Scale Factor=&VectorFromFloat(Value=$User.GlobalScale))
  ]
  rendererClasses [&NiagaraMeshRendererProperties]
  renderers [
   {rendererClass=&NiagaraMeshRendererProperties, rendererIndex=0}
   {rendererClass=&NiagaraMeshRendererProperties, rendererIndex=1}
  ]
 }}
summary {
 systemName=NS_Thermite_Pickup_Glow
 emitters [{emitterName=BG_Glow, rendererClasses [&NiagaraMeshRendererProperties]}]
 userVariables [User.GlobalScale {type=NiagaraFloat, defaultValue=1.2, description=""}]
}
