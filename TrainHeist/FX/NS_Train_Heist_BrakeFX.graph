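# Niagara brake effect for the Train Heist train, emitting skid flashes and sparks from eight partitioned car offsets.
# SkidSharp pairs short-lived skid particles with fake-light sprites on separate position and color attributes, using system-age curves to fade alpha.
# GPU Spark sprays curl-noise, speed-stretched sprites, dimmed by time of day and culled by a box.
# System brake-amount variables scale skid mesh size and spark spawn probability.
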NiagaraSystem /TrainHeist/FX/NS_Train_Heist_BrakeFX
dependencies {
 usedDataInterfaces [&NiagaraDataInterfaceArrayFloat3, &NiagaraDataInterfaceCurve, &NiagaraDataInterfaceColorCurve]
 usedDynamicInputs []
 usedModules [
  &SystemState
  &self:SystemSpawnScript.SystemScriptSource.SystemScriptGraph.NiagaraNodeAssignment_2.SetVariables_#1
  &self:SystemSpawnScript.SystemScriptSource.SystemScriptGraph.NiagaraNodeAssignment_0.SetVariables_#2
  &self:SystemSpawnScript.SystemScriptSource.SystemScriptGraph.NiagaraNodeAssignment_1.SetVariables_#3
 ]
}
emitters {
 SkidSharp {
  emitterUpdateScript [
   &EmitterState()
   &SpawnRate(SpawnRate=50, Use Spawn Probability=true)
   &SpawnBurst_Instantaneous(Spawn Count=8, Spawn Group=1, Use Spawn Probability=false, UseLoopCountLimit=false)
  ]
  particleSpawnScript [
   /SetVariables(Particles.SpawnGroup=$Emitter.SpawnGroup) {isSetParameters=true}
   /SetVariables(Particles.VisibilityTag=$Particles.SpawnGroup) {isSetParameters=true}
   &PartitionParticles(Number Of Partitions=8)
   &InitializeParticle(
    Color=<rgba 10 10 10 1>
    Lifetime=&MaskFloatBySpawnGroup(Masked Float=&RandomRangeFloat(Minimum=0.1, Maximum=0.15), Passthrough Float=86400)
    Mesh Scale=&RandomRangeVector(Minimum=<1 0.4 0.15>, Maximum=<1.5 0.8 0.25>)
    Mesh Scale Mode=::NewEnumerator4
    Sprite Size Mode=::NewEnumerator3
    Uniform Sprite Size=130
    UsePositionOffset=true
    Write Material Random=true
   ) {module=InitializeParticle001}
   &DynamicMaterialParameters(
    Index 0 Param 3=&RandomRangeFloat(Minimum=1, Maximum=1.5, Evaluation Type=::NewEnumerator1)
    Module.Param0WriteEnabled=false
    Module.Param1WriteEnabled=false
    Module.Param2WriteEnabled=true
    Module.Param3WriteEnabled=false
   ) {module=DynamicMaterialParameters001}
   &CameraOffset()
   &AddVelocity(Velocity=&SelectVectorFromArray(
     Direct Array Index=$Particles.PartitionParticles.Partition
     Direct Array Mode=::NewEnumerator1
     Vector Selection Array=DataInterface() {propertyValues {FloatData [<-1 0.25 0>, <-1 -0.25 0>]}}
    ))
   &SystemLocation(
    MaskSpawnGroup=false
    Offset=&SelectVectorFromArray(Direct Array Index=$Particles.PartitionParticles.Partition, Vector Selection Array=$System.Offset_Car)
   )
   &InitialMeshOrientation(Mesh Orientation Mode=::NewEnumerator2, Use Rotation Vector=true)
   /SetVariables(
    Particles.CameraOffset_FakeLit=20
    Particles.Color_FakeLit:LinearColor=&MultiplyLinearColor(Linear Color=<rgba 20 4.6667 0 1>, Scale Factor=<1 1 1 1>)
    Particles.Position_FakeLit:NiagaraPosition=&AddVectorToPosition(
     Position=$Particles.Position
     Vector=&TransformVector(
      Vector=&SelectVectorFromArray(
       Direct Array Index=$Particles.PartitionParticles.Partition
       Direct Array Mode=::NewEnumerator1
       Vector Selection Array=DataInterface() {propertyValues {FloatData [<0 30 10>, <0 -30 10>]}}
      )
      Destination Space=::Local
     )
    )
   ) {isSetParameters=true}
   /SetVariables(Particles.PositionSkid:NiagaraPosition=&AddVectorToPosition(Position=$Particles.Position, Vector=<-5 0 2>)) {isSetParameters=true}
  ]
  particleUpdateScript [
   &ParticleState()
   &ScaleColor(
    Scale Alpha=&FloatFromCurve(
     FloatCurve=DataInterface() {propertyValues {Curve {keys [
         {arriveTangent=-0.2241, interpMode=RCIM_Cubic, leaveTangent=-0.2241, tangentMode=RCTM_User, time=0, value=1}
         {arriveTangent=-3.1123, interpMode=RCIM_Cubic, leaveTangent=-3.1123, tangentMode=RCTM_User, time=1, value=0}
        ]}}}
     Scale Curve=&Lerp_Float(
      A=$User.breakingEffectAmount
      B=&FloatFromCurve(
       FloatCurve=DataInterface() {propertyValues {Curve {keys [
           {arriveTangent=4.2321, interpMode=RCIM_Cubic, leaveTangent=4.2321, tangentMode=RCTM_User, time=0.5, value=0}
           {arriveTangentWeight=0.3333, interpMode=RCIM_Cubic, leaveTangentWeight=1.3334, tangentMode=RCTM_User, time=1, value=1}
           {arriveTangentWeight=1.3333, interpMode=RCIM_Cubic, leaveTangentWeight=0.5, tangentMode=RCTM_Break, time=2, value=1}
           {arriveTangent=-0.2796, interpMode=RCIM_Cubic, leaveTangent=-0.2796, tangentMode=RCTM_User, time=3, value=0.15}
           {interpMode=RCIM_Cubic, tangentMode=RCTM_Break, time=4.7, value=0}
          ]}}}
       CurveIndex=$System.Age
      )
      Alpha=1
     )
    )
    Scale RGB=&MakeVectorFromLinearColorRGB(Color=&ColorFromCurve(ColorCurve=DataInterface() {propertyValues {
        AlphaCurve {keys [<at 0 1 RCIM_Cubic>]}
        BlueCurve {keys [<at 0.0137 0>,<at 0.1369 0.0045>,<at 0.9732 0>]}
        GreenCurve {keys [<at 0.0137 0.9333>,<at 0.1369 0.0623>,<at 0.9732 0.0583>]}
        RedCurve {keys [<at 0.0137 8>,<at 0.1369 0.5>,<at 0.9732 0.5>]}
       }}))
    ScaleA=true
    ScaleRGB=true
   )
   /SetVariables(Particles.Color_FakeLit:LinearColor=&MultiplyLinearColor(
     Linear Color=$Particles.Initial.Color_FakeLit
     Scale Factor=&MakeFloat4FromVectorAndFloat(
      "Vector (RGB)"=<1 1 1>
      "Float (Alpha)"=&Multiply_Float(
       A=1
       B=&Lerp_Float(
        A=$User.breakingEffectAmount
        B=&FloatFromCurve(
         FloatCurve=DataInterface() {propertyValues {Curve {keys [
             {arriveTangent=4.2321, interpMode=RCIM_Cubic, leaveTangent=4.2321, tangentMode=RCTM_User, time=0.5, value=0}
             {arriveTangentWeight=0.3333, interpMode=RCIM_Cubic, leaveTangentWeight=1.3334, tangentMode=RCTM_User, time=1, value=1}
             {arriveTangentWeight=1.3333, interpMode=RCIM_Cubic, leaveTangentWeight=0.5, tangentMode=RCTM_Break, time=2, value=1}
             {arriveTangent=-0.303, interpMode=RCIM_Cubic, leaveTangent=-0.303, time=3.9, value=0.15}
             {interpMode=RCIM_Cubic, tangentMode=RCTM_Break, time=5.3, value=0}
            ]}}}
         CurveIndex=$System.Age
        )
        Alpha=1
       )
      )
     )
    )) {isSetParameters=true}
   /SetVariables(Particles.DMP_FakeLight_01:Vector4f=&MakeVector4(X=0.5, Y=1, Z=1, W=0)) {isSetParameters=true}
   &DynamicMaterialParameters(
    Index 0 Param 1=&FloatFromCurve(FloatCurve=DataInterface() {propertyValues {Curve {keys [<at 0.1 0.1 RCIM_Cubic>,<at 0.8 1>]}}})
    Index 0 Param 2=2
    Index 0 Param 4=&FloatFromCurve(FloatCurve=DataInterface() {propertyValues {Curve {keys [<at 0 1>,<at 1 3>]}}})
    Module.Param0WriteEnabled=true
    Module.Param1WriteEnabled=true
    Module.Param2WriteEnabled=false
    Module.Param3WriteEnabled=true
   )
   &ScaleMeshSize(Scale Factor=&Multiply_VectorByFloat(
     Vector=&MakeVector(
      X=1
      Y=1
      Z=&FloatFromCurve(FloatCurve=DataInterface() {propertyValues {Curve {keys [
           {arriveTangent=2.4702, interpMode=RCIM_Cubic, leaveTangent=2.4702, tangentMode=RCTM_User, time=0, value=0}
           {arriveTangent=0.9995, interpMode=RCIM_Cubic, leaveTangent=0.9995, tangentMode=RCTM_User, time=0.4, value=0.75}
           {arriveTangent=-0.0405, interpMode=RCIM_Cubic, leaveTangent=-0.0405, tangentMode=RCTM_User, time=1, value=1}
          ]}}})
     )
     Float=&FloatFromCurve(
      FloatCurve=DataInterface() {propertyValues {Curve {keys [
          {arriveTangent=2.4702, interpMode=RCIM_Cubic, leaveTangent=2.4702, tangentMode=RCTM_User, time=0, value=0}
          {arriveTangent=0.9995, interpMode=RCIM_Cubic, leaveTangent=0.9995, tangentMode=RCTM_User, time=0.4, value=0.75}
          {arriveTangent=-0.0405, interpMode=RCIM_Cubic, leaveTangent=-0.0405, tangentMode=RCTM_User, time=1, value=1}
         ]}}}
      Scale Curve=&Multiply_Float(A=2, B=&Lerp_Float(A=0.2, B=1, Alpha=$System.breakingEffectAmount))
     )
    )) {module=ScaleMeshSize001}
   &SolveForcesAndVelocity(Clamp Velocity=false, Limit Acceleration=false)
   /SetVariables(Particles.TempScale:Vector3f=<5 5 5>) {isSetParameters=true}
  ]
  rendererClasses [&NiagaraMeshRendererProperties, &NiagaraSpriteRendererProperties]
  renderers [
   {rendererClass=&NiagaraMeshRendererProperties, rendererIndex=0}
   {rendererClass=&NiagaraSpriteRendererProperties, rendererIndex=1}
  ]
 }
 Spark {
  simTarget=GPUComputeSim
  emitterUpdateScript [
   &EmitterState()
   &SpawnRate(Spawn Probability=$System.breakingEffectSparksAmount, SpawnRate=1200, Use Spawn Probability=true)
  ]
  particleSpawnScript [
   &PartitionParticles(Number Of Partitions=8)
   &InitializeParticle(
    Color=<rgba 4 0.5231 0 1>
    Lifetime Min=0.4
    Lifetime Mode=::NewEnumerator1
    Sprite Size Max=<2.2 4>
    Sprite Size Min=<1.2 2>
    Sprite Size Mode=::NewEnumerator2
    UsePositionOffset=false
    Write Material Random=true
   )
   &SphereLocation(
    MaskSpawnGroup=false
    Offset=&SelectVectorFromArray(Direct Array Index=$Particles.PartitionParticles.Partition, Vector Selection Array=$System.Offset_Car)
    Override Seed=false
    Sphere Radius=&RandomRangeFloat(Minimum=1, Maximum=2)
    Surface Only=false
    UseNonUniformScale=true
    UseOffset=true
    UseSphereOrientation=false
   )
   &AddVelocity(
    Velocity=&SelectVectorFromArray(
     Direct Array Index=$Particles.PartitionParticles.Partition
     Direct Array Mode=::NewEnumerator1
     Vector Selection Array=DataInterface() {propertyValues {FloatData [<-1 0.33 0>, <-1 -0.33 0>]}}
    )
    Velocity Speed Scale=&RandomRangeFloat(Minimum=300, Maximum=1000)
   )
   &CameraOffset(Camera Offset Amount=50)
  ]
  particleUpdateScript [
   &CurlNoiseForce(
    Noise Strength=&Multiply_Float(A=&RandomRangeFloat(Minimum=2500, Maximum=5000), B=1)
    Pan Noise=true
    Pan Noise Field=<0.5 0.5 0.5>
   )
   &AccelerationForce(Acceleration=&Multiply_Vector(
     A=<0 0 0.05>
     B=&VectorFromFloat(Value=&RandomRangeFloat(Minimum=1500, Maximum=2500))
    ))
   &Drag(
    Drag=&FloatFromCurve(
     FloatCurve=DataInterface() {propertyValues {Curve {keys [
         {interpMode=RCIM_Cubic, tangentMode=RCTM_User, time=0, value=0}
         {arriveTangent=3.5282, interpMode=RCIM_Cubic, leaveTangent=3.5282, tangentMode=RCTM_User, time=1, value=1}
        ]}}}
     CurveIndex=&ReturnNormalizedExecIndex()
     Scale Curve=5
    )
    Use Linear Drag=true
    Use Rotational Drag=false
   )
   &SolveForcesAndVelocity(Clamp Velocity=false, Limit Acceleration=false)
   &ParticleState()
   &ScaleSpriteSize(Scale Factor=&Vector2DFromFloat(Value=&FloatFromCurve(FloatCurve=DataInterface() {propertyValues {Curve {keys [<at 0 0.2>,<at 0.2 1>,<at 1 0>]}}}, Scale Curve=2)))
   &ScaleSpriteSizeBySpeed(Max Scale Factor=<0.05 5>, Min Scale Factor=<1 0.5>)
   &ScaleColor(
    Scale Alpha=&FloatFromCurve(
     FloatCurve=DataInterface() {propertyValues {Curve {keys [<at 0 0>,<at 0.2 1>]}}}
     Scale Curve=&FNTimeOfDayScalar(Night Scale=0.6, Morning Scale=0.4, Evening Scale=0.4)
    )
    Scale RGB=&VectorFromFloat(Value=&Multiply_Float(
      A=&FloatFromCurve(FloatCurve=DataInterface() {propertyValues {Curve {keys [<at 0 50>,<at 1 10>]}}})
      B=&FloatFromCurve(FloatCurve=DataInterface() {propertyValues {Curve {keys [
           <at 0 1 RCIM_Cubic>
           {arriveTangent=2.7063, interpMode=RCIM_Cubic, leaveTangent=2.7063, time=0.2, value=1.3}
           {arriveTangent=-3.614, interpMode=RCIM_Cubic, leaveTangent=-3.614, time=0.3695, value=2}
           {arriveTangent=-1.4517, interpMode=RCIM_Cubic, leaveTangent=-1.4517, time=0.5434, value=0.059}
           {arriveTangent=1.0963, interpMode=RCIM_Cubic, leaveTangent=1.0963, time=0.7139, value=1.5}
           {arriveTangent=-1.7479, interpMode=RCIM_Cubic, leaveTangent=-1.7479, time=0.9, value=0.45}
           <at 1 1 RCIM_Cubic>
          ]}}})
     ))
    ScaleA=true
    ScaleRGB=true
   ) {module=ScaleColor001}
   &KillParticlesInVolume(Box Size:Vector3f=<1950 450 150>, Kill Shape=::NewEnumerator1, Origin Offset:Vector3f=<0 0 150>)
  ]
  rendererClasses [&NiagaraSpriteRendererProperties]
  renderers [{rendererClass=&NiagaraSpriteRendererProperties}]
 }
}
summary {
 systemName=NS_Train_Heist_BrakeFX
 emitters [
  {emitterName=Spark, simTarget=GPUComputeSim, rendererClasses [&NiagaraSpriteRendererProperties]}
  {emitterName=SkidSharp, rendererClasses [&NiagaraMeshRendererProperties, &NiagaraSpriteRendererProperties]}
 ]
 userVariables [User.breakingEffectAmount {type=NiagaraFloat, defaultValue=1, description=""}]
}
