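# Treasury car turret material for TrainHeist, textured with curved barricade diffuse, normal and packed specular/metallic/roughness maps.
# A HideTurret static switch sets World Position Offset to negated world position, collapsing the mesh to the world origin to hide it.
# A Mobile Opt composite substitutes constant specular and roughness on low-quality mobile, then feeds MF_CreativeMoveTool.
# It supports Nanite and instanced meshes, transforming normals out of tangent space.
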Material /TrainHeist/Materials/M_Treasury_Car_Turret_WPO
graph {
 properties {
  bAutomaticallySetUsageInEditor=false
  bTangentSpaceNormal=false
  bUseMaterialAttributes=true
  bUsedWithInstancedStaticMeshes=true
  bUsedWithNanite=true
  ditheredLODTransition=true
  pixelDepthOffsetMode=PDOM_Legacy
  shadingModel=MSM_FromMaterialExpression
  usedShadingModels=Default Lit
 }
 nodes [
  n1 = TextureSampleParameter2D() {parameterName=SpecularMasks, samplerType=SAMPLERTYPE_Masks, texture=/TrainHeist/Textures/T_Treasury_CurvedBarricade_B_A_S}
  n2 = WorldPosition()
  n3 = SetMaterialAttributes!(
   MaterialAttributes
   Base Color=TextureSampleParameter2D() {parameterName=Diffuse, texture=/TrainHeist/Textures/T_Treasury_CurvedBarricade_B_A_D}.RGB
   Metallic=NamedRerouteDeclaration(None=@n1.G) {name=IN_Metallic, nodeColor=<rgba 0 1 0.0391 1>}
   Specular=NamedRerouteDeclaration(None=@n1.R) {name=IN_Specular, nodeColor=<rgba 0.3594 1 0 1>}
   Roughness=NamedRerouteDeclaration(None=@n1.B) {name=IN_Roughness, nodeColor=<rgba 0.7578 0 1 1>}
   Normal=Transform(None=TextureSampleParameter2D() {parameterName=Normals, samplerType=SAMPLERTYPE_Normal, texture=/TrainHeist/Textures/T_Treasury_CurvedBarricade_B_A_N}.RGB)
   World Position Offset=StaticSwitchParameter(True=Multiply(A=@n2.XYZ) {constB=-1}, False=Constant()) {group=Turret_WPO, parameterName=HideTurret}
  ) {attributeSetTypes [
    #1
    #2
    #3
    #4
    #5
    #6
   ]}
  Composite!(InputPin=@n3) {subgraphName=Mobile Opt}
  MaterialExpressionReroute_7 = Reroute(None=@n3)
  n4 = GetMaterialAttributes(None=@MaterialExpressionReroute_7.InputPin)
  n5 = Constant() {r=0.5}
  n6 = Constant() {r=0.8}
  n7 = ShadingPathSwitch(
   Default=@MaterialExpressionReroute_7.InputPin
   Mobile=QualitySwitch(
    Default=@MaterialExpressionReroute_7.InputPin
    Low=SetMaterialAttributes!(MaterialAttributes=@n4.MaterialAttributes, Specular=@n5, Roughness=@n6) {attributeSetTypes [#3, #4]}
    High=@MaterialExpressionReroute_7.InputPin
    Medium=@MaterialExpressionReroute_7.InputPin
    Epic=@MaterialExpressionReroute_7.InputPin
   )
  )
  MaterialExpressionReroute_6 = Reroute(None=@n7)
  MaterialFunctionCall(MA In=@MaterialExpressionReroute_6.Output) {materialFunction=&MF_CreativeMoveTool}
  PinBase() {reroutePins [InputPin {expression=&self:MaterialExpressionReroute_7}]}
  PinBase!(Output=@n7) {reroutePins [Output {expression=&self:MaterialExpressionReroute_6}]}
 ]
}
